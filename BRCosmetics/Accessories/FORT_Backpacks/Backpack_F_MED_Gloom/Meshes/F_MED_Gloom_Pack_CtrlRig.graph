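# Control Rig for the Gloom backpack cosmetic that procedurally animates its six tentacle chains and adjusts the right eye.
# Chain Harmonics units run in sequence on each front, mid and back chain, layering wave and pendulum motion, with pendulum disabled on the mid-right chain.
# A spherical pose reader on eye_rgt then weights a local offset on eye_rgt_adj.
# A second back-right harmonics node is left unconnected, and debug drawing is on.
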ControlRigBlueprint /BRCosmetics/Accessories/FORT_Backpacks/Backpack_F_MED_Gloom/Meshes/F_MED_Gloom_Pack_CtrlRig
graphs [
 RigVMFunctionLibrary {links [], nodes []}
 RigVMModel {nodes [
   RigUnit_BeginExecution = &RigUnit_BeginExecution()
   /Script/ControlRig.RigUnit_ChainHarmonics(
    ChainRoot=tentacleFront_0_lft
    DrawWorldOffset=(Rotation=(X=0,Y=0,Z=0,W=1),Translation=(X=0,Y=0,Z=0),Scale3D=(X=1,Y=1,Z=1))
    Pendulum=(bEnabled=True,PendulumStiffness=2,PendulumGravity=(X=0,Y=0,Z=0),PendulumBlend=0.75,PendulumDrag=0.98,PendulumMaximum=0.1,UnwindAxis=(X=0,Y=1,Z=0),UnwindMinimum=0.2,UnwindMaximum=0.05)
    Reach=(bEnabled=True,ReachTarget=(X=0,Y=0,Z=0),ReachAxis=(X=1,Y=0,Z=0),ReachMinimum=0,ReachMaximum=0)
    Speed=(X=1,Y=1,Z=1)
    Wave=(bEnabled=True,WaveFrequency=(X=1,Y=0.6,Z=-3),WaveAmplitude=(X=0,Y=0,Z=1.5),WaveOffset=(X=0,Y=0,Z=2),WaveNoise=(X=0,Y=0,Z=0),WaveMinimum=0,WaveMaximum=0.5,WaveEase=Linear)
    WaveCurve=(EditorCurveData=(Keys=((Value=0.5),(Time=0.9986,Value=0.7527))))
    bDrawDebug=True
    ExecutePin=@RigUnit_BeginExecution.ExecutePin
   ) {is=struct RigVMUnitNode}
   > &RigUnit_ChainHarmonics(
    ChainRoot=tentacleMid_0_lft
    DrawWorldOffset=(Rotation=(X=0,Y=0,Z=0,W=1),Translation=(X=0,Y=0,Z=0),Scale3D=(X=1,Y=1,Z=1))
    Pendulum=(bEnabled=True,PendulumStiffness=2,PendulumGravity=(X=0,Y=0,Z=0),PendulumBlend=0.75,PendulumDrag=0.98,PendulumMaximum=0.1,UnwindAxis=(X=0,Y=1,Z=0),UnwindMinimum=0.2,UnwindMaximum=0.05)
    Reach=(bEnabled=True,ReachTarget=(X=0,Y=0,Z=0),ReachAxis=(X=1,Y=0,Z=0),ReachMinimum=0,ReachMaximum=0)
    Speed=(X=1,Y=1,Z=1)
    Wave=(bEnabled=True,WaveFrequency=(X=1,Y=0.6,Z=-3),WaveAmplitude=(X=0,Y=0,Z=0.5),WaveOffset=(X=0,Y=0,Z=2),WaveNoise=(X=0,Y=0,Z=0),WaveMinimum=0,WaveMaximum=0.5)
    WaveCurve=(EditorCurveData=(Keys=((Value=0.5),(Time=1,Value=0.75)),DefaultValue=340282346638528859811704183484516925440,PreInfinityExtrap=RCCE_Constant,PostInfinityExtrap=RCCE_Constant))
    bDrawDebug=True
   )
   > &RigUnit_ChainHarmonics(
    ChainRoot=tentacleBack_0_lft
    DrawWorldOffset=(Rotation=(X=0,Y=0,Z=0,W=1),Translation=(X=0,Y=0,Z=0),Scale3D=(X=1,Y=1,Z=1))
    Pendulum=(bEnabled=True,PendulumStiffness=2,PendulumGravity=(X=0,Y=0,Z=0),PendulumBlend=0.75,PendulumDrag=0.98,PendulumMaximum=0.1,UnwindAxis=(X=0,Y=1,Z=0),UnwindMinimum=0.2,UnwindMaximum=0.05)
    Reach=(bEnabled=True,ReachTarget=(X=0,Y=0,Z=0),ReachAxis=(X=1,Y=0,Z=0),ReachMinimum=0,ReachMaximum=0)
    Speed=(X=1,Y=1,Z=1)
    Wave=(bEnabled=True,WaveFrequency=(X=1,Y=0.6,Z=3),WaveAmplitude=(X=0,Y=0,Z=0.5),WaveOffset=(X=0,Y=0,Z=1.9),WaveNoise=(X=0,Y=0,Z=0),WaveMinimum=0,WaveMaximum=0.5)
    WaveCurve=(EditorCurveData=(Keys=((Value=0.5),(Time=1,Value=0.75)),DefaultValue=340282346638528859811704183484516925440,PreInfinityExtrap=RCCE_Constant,PostInfinityExtrap=RCCE_Constant))
    bDrawDebug=True
   )
   > &RigUnit_ChainHarmonics(
    ChainRoot=tentacleFront_0_rgt
    DrawWorldOffset=(Rotation=(X=0,Y=0,Z=0,W=1),Translation=(X=0,Y=0,Z=0),Scale3D=(X=1,Y=1,Z=1))
    Pendulum=(bEnabled=True,PendulumStiffness=2,PendulumGravity=(X=0,Y=0,Z=0),PendulumBlend=0.75,PendulumDrag=0.98,PendulumMaximum=0.1,UnwindAxis=(X=0,Y=1,Z=0),UnwindMinimum=0.2,UnwindMaximum=0.05)
    Reach=(bEnabled=True,ReachTarget=(X=0,Y=0,Z=0),ReachAxis=(X=1,Y=0,Z=0),ReachMinimum=0,ReachMaximum=0)
    Speed=(X=1,Y=1,Z=1)
    Wave=(bEnabled=True,WaveFrequency=(X=1,Y=0.6,Z=-3),WaveAmplitude=(X=0,Y=0,Z=1.5),WaveOffset=(X=0,Y=0,Z=2),WaveNoise=(X=0,Y=0,Z=0),WaveMinimum=0,WaveMaximum=0.5)
    WaveCurve=(EditorCurveData=(Keys=((Value=0.5),(Time=1,Value=0.75)),DefaultValue=340282346638528859811704183484516925440,PreInfinityExtrap=RCCE_Constant,PostInfinityExtrap=RCCE_Constant))
    bDrawDebug=True
   )
   > &RigUnit_ChainHarmonics(
    ChainRoot=tentacleMid_0_rgt
    DrawWorldOffset=(Rotation=(X=0,Y=0,Z=0,W=1),Translation=(X=0,Y=0,Z=0),Scale3D=(X=1,Y=1,Z=1))
    Pendulum=(bEnabled=false,PendulumStiffness=2,PendulumGravity=(X=0,Y=0,Z=0),PendulumBlend=0.75,PendulumDrag=0.98,PendulumMaximum=0.1,UnwindAxis=(X=0,Y=1,Z=0),UnwindMinimum=0.2,UnwindMaximum=0.05)
    Reach=(bEnabled=True,ReachTarget=(X=0,Y=0,Z=0),ReachAxis=(X=1,Y=0,Z=0),ReachMinimum=0,ReachMaximum=0)
    Speed=(X=1,Y=1,Z=1)
    Wave=(bEnabled=True,WaveFrequency=(X=1,Y=0.6,Z=-3),WaveAmplitude=(X=0,Y=0,Z=0.5),WaveOffset=(X=0,Y=0,Z=2.1),WaveNoise=(X=0,Y=0,Z=0),WaveMinimum=0,WaveMaximum=0.5)
    WaveCurve=(EditorCurveData=(Keys=((Value=0.5),(Time=1,Value=0.75)),DefaultValue=340282346638528859811704183484516925440,PreInfinityExtrap=RCCE_Constant,PostInfinityExtrap=RCCE_Constant))
    bDrawDebug=True
   )
   > &RigUnit_ChainHarmonics(
    ChainRoot=tentacleBack_0_rgt
    DrawWorldOffset=(Rotation=(X=0,Y=0,Z=0,W=1),Translation=(X=0,Y=0,Z=0),Scale3D=(X=1,Y=1,Z=1))
    Pendulum=(bEnabled=True,PendulumStiffness=2,PendulumGravity=(X=0,Y=0,Z=0),PendulumBlend=0.75,PendulumDrag=0.98,PendulumMaximum=0.1,UnwindAxis=(X=0,Y=1,Z=0),UnwindMinimum=0.2,UnwindMaximum=0.05)
    Reach=(bEnabled=True,ReachTarget=(X=0,Y=0,Z=0),ReachAxis=(X=1,Y=0,Z=0),ReachMinimum=0,ReachMaximum=0)
    Speed=(X=1,Y=1,Z=1)
    Wave=(bEnabled=True,WaveFrequency=(X=1,Y=0.6,Z=3),WaveAmplitude=(X=0,Y=0,Z=0.5),WaveOffset=(X=0,Y=0,Z=1.9),WaveNoise=(X=0,Y=0,Z=0),WaveMinimum=0,WaveMaximum=0.5)
    WaveCurve=(EditorCurveData=(Keys=((Value=0.5),(Time=1,Value=0.75)),DefaultValue=340282346638528859811704183484516925440,PreInfinityExtrap=RCCE_Constant,PostInfinityExtrap=RCCE_Constant))
    bDrawDebug=True
   )
   > n1 = &RigUnit_SphericalPoseReader(
    ActiveRegionScaleFactors=(PositiveWidth=1,NegativeWidth=1,PositiveHeight=1,NegativeHeight=1)
    ActiveRegionSize=0.1
    Debug=(bDrawDebug=True,DebugScale=25,DebugSegments=20,DebugThickness=0.25)
    DriverAxis=(X=-1,Y=0,Z=0)
    DriverItem=(Type=Bone,Name="eye_rgt")
    FalloffRegionScaleFactors=(PositiveWidth=1,NegativeWidth=1,PositiveHeight=1,NegativeHeight=1)
    FalloffSize=0.1
    OptionalParentItem=(Type=Bone,Name="root")
    RotationOffset=(X=0,Y=90,Z=32)
   )
   > &RigUnit_SetTransform(
    Item=(Type=Bone,Name="eye_rgt_adj")
    Space=LocalSpace
    Value=(Rotation=(X=0,Y=0,Z=0,W=1),Translation=(X=0,Y=0.5,Z=0),Scale3D=(X=1,Y=1,Z=1))
    Weight=@n1.OutputParam
   )
   &RigUnit_ChainHarmonics(
    ChainRoot=tentacleBack_0_rgt
    DrawWorldOffset=(Rotation=(X=0,Y=0,Z=0,W=1),Translation=(X=0,Y=0,Z=0),Scale3D=(X=1,Y=1,Z=1))
    ExecutePin=()
    Pendulum=(bEnabled=True,PendulumStiffness=2,PendulumGravity=(X=0,Y=0,Z=0),PendulumBlend=0.75,PendulumDrag=0.98,PendulumMaximum=0.1,UnwindAxis=(X=0,Y=1,Z=0),UnwindMinimum=0.2,UnwindMaximum=0.05)
    Reach=(bEnabled=True,ReachTarget=(X=0,Y=0,Z=0),ReachAxis=(X=1,Y=0,Z=0),ReachMinimum=0,ReachMaximum=0)
    Speed=(X=1,Y=1,Z=1)
    Wave=(bEnabled=True,WaveFrequency=(X=1,Y=0.6,Z=1.5),WaveAmplitude=(X=0,Y=0,Z=1.5),WaveOffset=(X=0,Y=0,Z=2.4),WaveNoise=(X=0,Y=0,Z=0),WaveMinimum=0,WaveMaximum=1.5)
    WaveCurve=(EditorCurveData=(Keys=((Value=0.5),(Time=1,Value=0.75)),DefaultValue=340282346638528859811704183484516925440,PreInfinityExtrap=RCCE_Constant,PostInfinityExtrap=RCCE_Constant))
    bDrawDebug=True
   )
  ]}
]
hierarchy [
 root {loc=<0 0 0>, rot=<0 0 0 1>}
 pack {parent=root, loc=<0 -0.7316 0>, rot=<0 0 0 1>}
 head_0 {parent=pack, loc=<0 -7.32 4.5325>, rot=<0.1959 -0.6794 0.1959 -0.6794>}
 head_1 {parent=head_0, loc=<-9.7585 0 0>, rot=<0 0 0 1>}
 eye_lft {parent=head_0, loc=<-2.2623 -3.1463 1.3428>, rot=<-0.6189 0.3419 0.3419 -0.6189>}
 eye_lft_adj {parent=eye_lft, loc=<0 0 0>, rot=<0 0 0 1>}
 eye_rgt {parent=head_0, loc=<-2.2623 -3.1463 -1.3428>, rot=<-0.3419 -0.6189 -0.6189 -0.3419>}
 eye_rgt_adj {parent=eye_rgt, loc=<0 0 0>, rot=<0 0 0 1>}
 eyeLid_up_lft {parent=head_0, loc=<-2.2623 -3.1463 1.3428>, rot=<0.1461 0.0421 -0.2738 -0.9497>}
 eyeLid_dwn_lft {parent=head_0, loc=<-2.2623 -3.1463 1.3428>, rot=<0.1461 0.0421 -0.2738 -0.9497>}
 eyeLid_up_rgt {parent=head_0, loc=<-2.2623 -3.1463 -1.3428>, rot=<0.0421 -0.1461 0.9497 -0.2738>}
 eyeLid_dwn_rgt {parent=head_0, loc=<-2.2623 -3.1463 -1.3428>, rot=<0.0421 -0.1461 0.9497 -0.2738>}
 tentacleFront_0_lft_FK {parent=pack, loc=<0.9483 -8.8682 2.9949>, rot=<0 -0.7071 0 -0.7071>}
 tentacleFront_0_lft_IK {parent=pack, loc=<0.9483 -8.8682 2.9949>, rot=<0 -0.7071 0 -0.7071>}
 tentacleFront_0_lft {parent=pack, loc=<0.9483 -8.8682 2.9949>, rot=<0 -0.7071 0 -0.7071>}
 tentacleFront_1_lft {parent=tentacleFront_0_lft, loc=<1.5203 0 0>, rot=<0 0 0 1>}
 tentacleFront_2_lft {parent=tentacleFront_1_lft, loc=<2.4524 0 0>, rot=<0 0 0 1>}
 tentacleFront_3_lft {parent=tentacleFront_2_lft, loc=<2.9156 0 0>, rot=<0 0 0 1>}
 tentacleFront_4_lft {parent=tentacleFront_3_lft, loc=<3.2253 0 0>, rot=<0 0 0 1>}
 tentacleFront_5_lft {parent=tentacleFront_4_lft, loc=<3.4684 0 0>, rot=<0 0 0 1>}
 tentacleFront_6_lft {parent=tentacleFront_5_lft, loc=<3.2274 0 0>, rot=<0 0 0 1>}
 tentacleFront_7_lft {parent=tentacleFront_6_lft, loc=<3.0425 0 0>, rot=<0 0 0 1>}
 tentacleFront_8_lft {parent=tentacleFront_7_lft, loc=<3.3791 0 0>, rot=<0 0 0 1>}
 tentacleFront_9_lft {parent=tentacleFront_8_lft, loc=<1.7025 0 0>, rot=<0 0 0 1>}
 tentacleMid_0_lft_FK {parent=pack, loc=<2.2061 -7.2492 2.9949>, rot=<0 -0.7071 0 -0.7071>}
 tentacleMid_0_lft_IK {parent=pack, loc=<2.2061 -7.2492 2.9949>, rot=<0 -0.7071 0 -0.7071>}
 tentacleMid_0_lft {parent=pack, loc=<2.2061 -7.2492 2.9949>, rot=<0.5 -0.5 -0.5 -0.5>}
 tentacleMid_1_lft {parent=tentacleMid_0_lft, loc=<1.5203 0 0>, rot=<0 0 0 1>}
 tentacleMid_2_lft {parent=tentacleMid_1_lft, loc=<2.4524 0 0>, rot=<0 0 0 1>}
 tentacleMid_3_lft {parent=tentacleMid_2_lft, loc=<2.9156 0 0>, rot=<0 0 0 1>}
 tentacleMid_4_lft {parent=tentacleMid_3_lft, loc=<3.2253 0 0>, rot=<0 0 0 1>}
 tentacleMid_5_lft {parent=tentacleMid_4_lft, loc=<3.4684 0 0>, rot=<0 0 0 1>}
 tentacleMid_6_lft {parent=tentacleMid_5_lft, loc=<3.2274 0 0>, rot=<0 0 0 1>}
 tentacleMid_7_lft {parent=tentacleMid_6_lft, loc=<3.0425 0 0>, rot=<0 0 0 1>}
 tentacleMid_8_lft {parent=tentacleMid_7_lft, loc=<3.3791 0 0>, rot=<0 0 0 1>}
 tentacleMid_9_lft {parent=tentacleMid_8_lft, loc=<1.7025 0 0>, rot=<0 0 0 1>}
 tentacleBack_0_lft_FK {parent=pack, loc=<1.124 -5.5782 2.9949>, rot=<-0.2076 -0.676 0.2076 -0.676>}
 tentacleBack_0_lft_IK {parent=pack, loc=<1.124 -5.5782 2.9949>, rot=<-0.2076 -0.676 0.2076 -0.676>}
 tentacleBack_0_lft {parent=pack, loc=<1.124 -5.5782 2.9949>, rot=<-0.2076 -0.676 0.2076 -0.676>}
 tentacleBack_1_lft {parent=tentacleBack_0_lft, loc=<1.5203 0 0>, rot=<0 0 0 1>}
 tentacleBack_2_lft {parent=tentacleBack_1_lft, loc=<2.4524 0 0>, rot=<0 0 0 1>}
 tentacleBack_3_lft {parent=tentacleBack_2_lft, loc=<2.9156 0 0>, rot=<0 0 0 1>}
 tentacleBack_4_lft {parent=tentacleBack_3_lft, loc=<3.2253 0 0>, rot=<0 0 0 1>}
 tentacleBack_5_lft {parent=tentacleBack_4_lft, loc=<3.4684 0 0>, rot=<0 0 0 1>}
 tentacleBack_6_lft {parent=tentacleBack_5_lft, loc=<3.2274 0 0>, rot=<0 0 0 1>}
 tentacleBack_7_lft {parent=tentacleBack_6_lft, loc=<3.0425 0 0>, rot=<0 0 0 1>}
 tentacleBack_8_lft {parent=tentacleBack_7_lft, loc=<3.3791 0 0>, rot=<0 0 0 1>}
 tentacleBack_9_lft {parent=tentacleBack_8_lft, loc=<1.7025 0 0>, rot=<0 0 0 1>}
 tentacleFront_0_rgt_FK {parent=pack, loc=<-0.9483 -8.8681 2.9949>, rot=<0.7071 0 0.7071 0>}
 tentacleFront_0_rgt_IK {parent=pack, loc=<-0.9483 -8.8682 2.9949>, rot=<0.7071 0 0.7071 0>}
 tentacleFront_0_rgt {parent=pack, loc=<-0.9483 -8.8682 2.9949>, rot=<0.7071 0 0.7071 0>}
 tentacleFront_1_rgt {parent=tentacleFront_0_rgt, loc=<-1.5204 0 0>, rot=<0 0 0 1>}
 tentacleFront_2_rgt {parent=tentacleFront_1_rgt, loc=<-2.4523 0 0>, rot=<0 0 0 1>}
 tentacleFront_3_rgt {parent=tentacleFront_2_rgt, loc=<-2.9157 0 0>, rot=<0 0 0 1>}
 tentacleFront_4_rgt {parent=tentacleFront_3_rgt, loc=<-3.2252 0 0>, rot=<0 0 0 1>}
 tentacleFront_5_rgt {parent=tentacleFront_4_rgt, loc=<-3.4684 0 0>, rot=<0 0 0 1>}
 tentacleFront_6_rgt {parent=tentacleFront_5_rgt, loc=<-3.2275 0 0>, rot=<0 0 0 1>}
 tentacleFront_7_rgt {parent=tentacleFront_6_rgt, loc=<-3.0425 0 0>, rot=<0 0 0 1>}
 tentacleFront_8_rgt {parent=tentacleFront_7_rgt, loc=<-3.3791 0 0>, rot=<0 0 0 1>}
 tentacleFront_9_rgt {parent=tentacleFront_8_rgt, loc=<-1.7025 0 0>, rot=<0 0 0 1>}
 tentacleMid_0_rgt_FK {parent=pack, loc=<-2.2061 -7.2492 2.9949>, rot=<0.7071 0 0.7071 0>}
 tentacleMid_0_rgt_IK {parent=pack, loc=<-2.2061 -7.2492 2.9949>, rot=<0.7071 0 0.7071 0>}
 tentacleMid_0_rgt {parent=pack, loc=<-2.2061 -7.2492 2.9949>, rot=<-0.5 0.5 -0.5 -0.5>}
 tentacleMid_1_rgt {parent=tentacleMid_0_rgt, loc=<-1.5204 0 0>, rot=<0 0 0 1>}
 tentacleMid_2_rgt {parent=tentacleMid_1_rgt, loc=<-2.4523 0 0>, rot=<0 0 0 1>}
 tentacleMid_3_rgt {parent=tentacleMid_2_rgt, loc=<-2.9157 0 0>, rot=<0 0 0 1>}
 tentacleMid_4_rgt {parent=tentacleMid_3_rgt, loc=<-3.2252 0 0>, rot=<0 0 0 1>}
 tentacleMid_5_rgt {parent=tentacleMid_4_rgt, loc=<-3.4684 0 0>, rot=<0 0 0 1>}
 tentacleMid_6_rgt {parent=tentacleMid_5_rgt, loc=<-3.2275 0 0>, rot=<0 0 0 1>}
 tentacleMid_7_rgt {parent=tentacleMid_6_rgt, loc=<-3.0425 0 0>, rot=<0 0 0 1>}
 tentacleMid_8_rgt {parent=tentacleMid_7_rgt, loc=<-3.3791 0 0>, rot=<0 0 0 1>}
 tentacleMid_9_rgt {parent=tentacleMid_8_rgt, loc=<-1.7025 0 0>, rot=<0 0 0 1>}
 tentacleBack_0_rgt_FK {parent=pack, loc=<-1.124 -5.5782 2.9949>, rot=<0.676 0.2076 0.676 -0.2076>}
 tentacleBack_0_rgt_IK {parent=pack, loc=<-1.124 -5.5782 2.9949>, rot=<0.676 0.2076 0.676 -0.2076>}
 tentacleBack_0_rgt {parent=pack, loc=<-1.124 -5.5782 2.9949>, rot=<0.676 0.2076 0.676 -0.2076>}
 tentacleBack_1_rgt {parent=tentacleBack_0_rgt, loc=<-1.5204 0 0>, rot=<0 0 0 1>}
 tentacleBack_2_rgt {parent=tentacleBack_1_rgt, loc=<-2.4523 0 0>, rot=<0 0 0 1>}
 tentacleBack_3_rgt {parent=tentacleBack_2_rgt, loc=<-2.9157 0 0>, rot=<0 0 0 1>}
 tentacleBack_4_rgt {parent=tentacleBack_3_rgt, loc=<-3.2252 0 0>, rot=<0 0 0 1>}
 tentacleBack_5_rgt {parent=tentacleBack_4_rgt, loc=<-3.4684 0 0>, rot=<0 0 0 1>}
 tentacleBack_6_rgt {parent=tentacleBack_5_rgt, loc=<-3.2275 0 0>, rot=<0 0 0 1>}
 tentacleBack_7_rgt {parent=tentacleBack_6_rgt, loc=<-3.0425 0 0>, rot=<0 0 0 1>}
 tentacleBack_8_rgt {parent=tentacleBack_7_rgt, loc=<-3.3791 0 0>, rot=<0 0 0 1>}
 tentacleBack_9_rgt {parent=tentacleBack_8_rgt, loc=<-1.7025 0 0>, rot=<0 0 0 1>}
]
