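# Animation Blueprint for the Gloom backpack cosmetic, animating its tentacled mesh with idle motion.
# A Random Sequence Player favors the idle loop over an occasional idle break, then additive bone-space Modify Bone nodes offset the left and right front, mid and back tentacle root bones.
# The adjusted pose feeds the pack's Control Rig before output.
# An unconnected "Animation Testing" state machine alternates the same two clips, and the event graph holds no logic.
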AnimBlueprint /BRCosmetics/Accessories/FORT_Backpacks/Backpack_F_MED_Gloom/Meshes/F_MED_Gloom_Pack_AnimBP
target_skeleton=&/F_MED_Gloom_Pack_Skeleton
graphs [
 AnimGraph {nodes [
   Local Space Ref Pose()
   n1 = Random Sequence Player!() -> (Pose:Pose Link Structure) {
    is=title AnimGraphNode_RandomPlayer
    entries [
     {
      chanceToPlay=0.8
      maxLoopCount=0
      maxPlayRate=1
      minLoopCount=0
      minPlayRate=1
      sequence=/BRCosmetics/Accessories/FORT_Backpacks/Backpack_F_MED_Gloom/Animation/Gloom_Female_Idle_BACKPACK
      blendIn {blendOption=Linear, blendTime=0.2, customCurve=null}
     }
     {
      chanceToPlay=0.2
      maxLoopCount=0
      maxPlayRate=1
      minLoopCount=0
      minPlayRate=1
      sequence=/BRCosmetics/Accessories/FORT_Backpacks/Backpack_F_MED_Gloom/Animation/Gloom_Female_Idle_Break_BACKPACK
      blendIn {blendOption=Linear, blendTime=0.2, customCurve=null}
     }
    ]
   }
   n2 = "Transform (Modify) Bone - Bone: tentacleFront_0_lft"(ComponentPose=Local To Component(LocalPose=@n1), Alpha=1, Translation="0,0,0.5", Rotation="0,-15,20", Scale="1,1,1") {
    rotationMode=BMM_Additive
    rotationSpace=BCS_BoneSpace
    scaleMode=BMM_Additive
    scaleSpace=BCS_BoneSpace
    translationMode=BMM_Additive
    translationSpace=BCS_BoneSpace
    alphaBoolBlend {blendOption=Linear}
    boneToModify {boneName=tentacleFront_0_lft}
    rotation {roll=20, yaw=-15}
    translation {z=0.5}
   }
   n3 = "Transform (Modify) Bone - Bone: tentacleMid_0_lft"(ComponentPose=@n2, Alpha=1, Translation="0,0,0", Rotation="0,-10,0", Scale="1,1,1") {
    rotationMode=BMM_Additive
    rotationSpace=BCS_BoneSpace
    scaleMode=BMM_Additive
    scaleSpace=BCS_BoneSpace
    translationMode=BMM_Additive
    translationSpace=BCS_BoneSpace
    alphaBoolBlend {blendOption=Linear}
    boneToModify {boneName=tentacleMid_0_lft}
    rotation {yaw=-10}
   }
   n4 = "Transform (Modify) Bone - Bone: tentacleBack_0_lft"(ComponentPose=@n3, Alpha=1, Translation="0,0,0", Rotation="0,15,0", Scale="1,1,1") {
    rotationMode=BMM_Additive
    rotationSpace=BCS_BoneSpace
    scaleMode=BMM_Additive
    scaleSpace=BCS_BoneSpace
    translationMode=BMM_Additive
    translationSpace=BCS_BoneSpace
    alphaBoolBlend {blendOption=Linear}
    boneToModify {boneName=tentacleBack_0_lft}
    rotation {yaw=15}
   }
   n5 = "Transform (Modify) Bone - Bone: tentacleFront_0_rgt"(ComponentPose=@n4, Alpha=1, Translation="0,0,-0.5", Rotation="0,-15,20", Scale="1,1,1") {
    rotationMode=BMM_Additive
    rotationSpace=BCS_BoneSpace
    scaleMode=BMM_Additive
    scaleSpace=BCS_BoneSpace
    translationMode=BMM_Additive
    translationSpace=BCS_BoneSpace
    alphaBoolBlend {blendOption=Linear}
    boneToModify {boneName=tentacleFront_0_rgt}
    rotation {roll=20, yaw=-15}
    translation {z=-0.5}
   }
   n6 = "Transform (Modify) Bone - Bone: tentacleMid_0_rgt"(ComponentPose=@n5, Alpha=1, Translation="0,0,0", Rotation="0,-10,0", Scale="1,1,1") {
    rotationMode=BMM_Additive
    rotationSpace=BCS_BoneSpace
    scaleMode=BMM_Additive
    scaleSpace=BCS_BoneSpace
    translationMode=BMM_Additive
    translationSpace=BCS_BoneSpace
    alphaBoolBlend {blendOption=Linear}
    boneToModify {boneName=tentacleMid_0_rgt}
    rotation {yaw=-10}
   }
   n7 = "Transform (Modify) Bone - Bone: tentacleBack_0_rgt"(ComponentPose=@n6, Alpha=1, Translation="0,0,0", Rotation="0,15,0", Scale="1,1,1") {
    rotationMode=BMM_Additive
    rotationSpace=BCS_BoneSpace
    scaleMode=BMM_Additive
    scaleSpace=BCS_BoneSpace
    translationMode=BMM_Additive
    translationSpace=BCS_BoneSpace
    alphaBoolBlend {blendOption=Linear}
    boneToModify {boneName=tentacleBack_0_rgt}
    rotation {yaw=15}
   }
   n8 = Control Rig(Source=Component To Local(ComponentPose=@n7)) {
    alphaBoolBlend {blendOption=Linear}
    controlRigAssetReference {blueprintRigClass=&/F_MED_Gloom_Pack_CtrlRig.F_MED_Gloom_Pack_CtrlRig_C}
   }
   Output Pose(Result=@n8) {layerGroup=None}
   AnimGraphNode_StateMachine_0 = Animation Testing() {is=title AnimGraphNode_StateMachine}
  ]}
 Animation Testing {
  entry=State
  path=AnimGraph.AnimGraphNode_StateMachine_0.Animation Testing
  ids=id
  nodes [
   AnimStateNode_0 = State() {is=title AnimStateNode}
   AnimStateNode_1 = State_1() {is=title AnimStateNode}
   AnimStateTransitionNode_0 = AnimStateTransitionNode() {bAutomaticRuleBasedOnSequencePlayerInState=true, blendMode=Linear, crossfadeDuration=0}
   AnimStateTransitionNode_1 = AnimStateTransitionNode() {bAutomaticRuleBasedOnSequencePlayerInState=true, blendMode=Linear, crossfadeDuration=0}
  ]
  links [
   {
    dst=State_1
    rule=AnimGraph.AnimGraphNode_StateMachine_0.Animation Testing.AnimStateTransitionNode_0.Transition
    src=State
    via=AnimStateTransitionNode_0
   }
   {
    dst=State
    rule=AnimGraph.AnimGraphNode_StateMachine_0.Animation Testing.AnimStateTransitionNode_1.Transition
    src=State_1
    via=AnimStateTransitionNode_1
   }
  ]
 }
 State {
  path=AnimGraph.AnimGraphNode_StateMachine_0.Animation Testing.AnimStateNode_0.State
  nodes [Output Animation Pose(Result=Sequence Player 'Gloom_Female_Idle_BACKPACK'() {bLoopAnimation=false, sequence=&Gloom_Female_Idle_BACKPACK})]
 }
 State_1 {
  path=AnimGraph.AnimGraphNode_StateMachine_0.Animation Testing.AnimStateNode_1.State_1
  nodes [
   n1 = Sequence Player 'Gloom_Female_Idle_Break_BACKPACK'() {bLoopAnimation=false, sequence=&Gloom_Female_Idle_Break_BACKPACK}
   Output Animation Pose(Result=@n1)
  ]
 }
 Transition {path=AnimGraph.AnimGraphNode_StateMachine_0.Animation Testing.AnimStateTransitionNode_0.Transition, nodes [Result()]}
 Transition {path=AnimGraph.AnimGraphNode_StateMachine_0.Animation Testing.AnimStateTransitionNode_1.Transition, nodes [Result()]}
 EventGraph {nodes [
   TryGetPawnOwner()
   Event BlueprintUpdateAnimation()
  ]}
]
variables []
